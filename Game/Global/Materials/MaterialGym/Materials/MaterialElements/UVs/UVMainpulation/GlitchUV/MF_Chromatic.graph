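# Material function that produces a chromatic glitch effect on an input texture at given UVs, outputting RGB.
# A sine of time, scaled by Chromatic_Speed and Chromatic_Strength parameters, multiplies the red and green channels of a panning glitch texture to form an animated UV offset.
# Only the green channel is sampled at the offset UVs; red and blue come from the undistorted sample, creating the color split.
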MaterialFunction /Game/Global/Materials/MaterialGym/Materials/MaterialElements/UVs/UVMainpulation/GlitchUV/MF_Chromatic
graph {nodes [
  n1 = FunctionInput(Preview=TextureCoordinate()) {bUsePreviewValueAsDefault=true, inputName=UV, inputType=FunctionInput_Vector2}
  n2 = FunctionInput(Preview=TextureObjectParameter() {parameterName=Texture, texture=/UVEditor/Textures/UVEditorColorGrid_Color}) {bUsePreviewValueAsDefault=true, inputName=Texture, inputType=FunctionInput_Texture2D}
  n3 = TextureSample(UVs=@n1, Tex=@n2) {texture=&UVEditorColorGrid_Color}
  n4 = Time()
  n5 = ScalarParameter() {defaultValue=10, group=Glitch_Charmatic, parameterName=Chromatic_Speed}
  n6 = Multiply(A=Sine(None=Multiply(A=@n4, B=@n5)), B=ScalarParameter() {defaultValue=0.006, group=Glitch_Charmatic, parameterName=Chromatic_Strength}) {constB=0.01}
  n7 = TextureSampleParameter2D(UVs=Panner(Coordinate=@n1) {speedX=0.1, speedY=0.1}) {
   parameterName=GlitchUVTexture
   texture=/Game/Accessories/FORT_Backpacks/Backpack_M_MED_FN_Birthday_5/Textures/TechArt/T_M_MED_FN_Birthday_5_Backpack_Glitch_Confetti_01
  }
  n8 = ComponentMask(None=@n7.RGB) {g=true, r=true}
  n9 = TextureSample(UVs=Add(A=@n1, B=Multiply(A=@n6, B=@n8)), Tex=@n2) {texture=&UVEditorColorGrid_Color}
  FunctionOutput(None=AppendVector(A=AppendVector(A=@n3.R, B=@n9.G), B=@n3.B))
 ]}
