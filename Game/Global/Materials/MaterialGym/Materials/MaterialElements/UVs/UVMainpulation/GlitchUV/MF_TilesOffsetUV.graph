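# Material function in the MaterialGym GlitchUV set that outputs slightly distorted texture coordinates for an animated glitch effect.
# It samples the GlitchUVTexture parameter, defaulting to a Birthday 5 backpack glitch confetti texture, and uses its red and green channels as alternate UVs.
# A fast time-driven sine, scaled very small, lerps the base UVs toward and away from those values, making the offset oscillate subtly.
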MaterialFunction /Game/Global/Materials/MaterialGym/Materials/MaterialElements/UVs/UVMainpulation/GlitchUV/MF_TilesOffsetUV
graph {nodes [
  n1 = TextureSampleParameter2D(UVs=TextureCoordinate()) {
   parameterName=GlitchUVTexture
   texture=/Game/Accessories/FORT_Backpacks/Backpack_M_MED_FN_Birthday_5/Textures/TechArt/T_M_MED_FN_Birthday_5_Backpack_Glitch_Confetti_01
  }
  n2 = Sine(None=Multiply(A=Time()) {constB=5})
  FunctionOutput(None=LinearInterpolate(A=TextureCoordinate(), B=ComponentMask(None=@n1.RGB) {g=true, r=true}, Alpha=Multiply(A=@n2) {constB=0.01}))
 ]}
