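# Material Gym element that demonstrates glitch-style UV manipulation in a Default Lit material flagged for pixel animation.
# MF_GlitchFunction, from the weapon skin function library, supplies glitched UVs, which are lerped with the output of MF_TilesOffsetUV by the GlitchUVBlend parameter.
# The blended UV drives MF_Chromatic, which by its name adds a chromatic split to the texture lookup.
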Material /Game/Global/Materials/MaterialGym/Materials/MaterialElements/UVs/UVMainpulation/GlitchUV/M_GlitchUV
graph {
 properties {bAutomaticallySetUsageInEditor=false, bHasPixelAnimation=true, usedShadingModels=Default Lit}
 nodes [MaterialFunctionCall!(
   UV=LinearInterpolate(
    A=MaterialFunctionCall!(UV) {materialFunction=/Game/Weapons/WeaponSkins/MaterialFunction/MF_GlitchFunction}.Glitched UV
    B=MaterialFunctionCall() {materialFunction=&/MF_TilesOffsetUV}.Result
    Alpha=ScalarParameter() {defaultValue=0.5, group=Glitch_UV, parameterName=GlitchUVBlend}
   )
   Texture
  ) {materialFunction=&/MF_Chromatic}]
}
